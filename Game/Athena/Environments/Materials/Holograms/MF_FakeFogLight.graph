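# Material function for a translucent fake fog light, faking a glowing point light and its haze, tinted by particle color.
# Opacity sums scene-depth surface lighting, shaped by world normals on deferred, and an analytic ray-sphere fog falloff, scaled by distance fades and time of day.
# A static bool toggles depth fade and outside- versus inside-sphere visibility.
# World position offset collapses vertices onto the camera when hidden, and always on forward and mobile.
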MaterialFunction /Game/Athena/Environments/Materials/Holograms/MF_FakeFogLight
graph {nodes [
  n1 = Multiply(A=ParticleColor().RGB, B=VectorParameter() {parameterName=Color, defaultValue=<rgba 1 0.168 0.0312 1>}.RGB)
  n2 = ScalarParameter() {defaultValue=5, parameterName=Brightness}
  n3 = ShadingPathSwitch(Default=@n2, Deferred=@n2, Forward=Multiply(A=@n2) {constB=0.35})
  n4 = ParticleColor()
  n5 = MaterialFunctionCall() {materialFunction=&WorldPositionBehindTranslucency}
  n6 = CameraPositionWS()
  n7 = Clamp(None=Divide(A=Distance(A=@n5.Result, B=@n6), B=ScalarParameter() {defaultValue=2000, parameterName=NearFadeDistance}))
  n8 = Distance(A=ObjectPositionWS(), B=CameraPositionWS())
  n9 = ScalarParameter() {defaultValue=15000, parameterName=CameraFadeOffset}
  n10 = Clamp(None=Divide(A=Subtract(A=@n8, B=@n9), B=ScalarParameter() {defaultValue=10000, parameterName=CameraFadeLength}))
  n11 = OneMinus(None=@n10)
  n12 = Distance(
   A=MaterialFunctionCall() {materialFunction=&WorldPositionBehindTranslucency}.Result
   B=ObjectPositionWS()
  )
  n13 = ScalarParameter() {defaultValue=2706.8237, parameterName=LightRadius}
  n14 = OneMinus(None=Clamp(None=Divide(A=@n12, B=@n13)))
  n15 = SceneTexture() {sceneTextureId=PPI_WorldNormal}
  n16 = Subtract(
   A=ObjectPositionWS()
   B=MaterialFunctionCall() {materialFunction=&WorldPositionBehindTranslucency}.Result
  )
  n17 = LinearInterpolate(
   A=ScalarParameter() {defaultValue=1, parameterName=MinLightTop}
   Alpha=DotProduct(A=ComponentMask(None=@n15.Color) {b=true, g=true, r=true}, B=Normalize(VectorInput=@n16))
  ) {constA=1}
  n18 = Constant() {r=1}
  n19 = ShadingPathSwitch(Default=@n17, Deferred=@n17, Forward=@n18, Mobile=@n18)
  n20 = CameraPositionWS()
  n21 = Subtract(A=WorldPosition().XYZ, B=@n20) {constB=-1}
  n22 = OneMinus(None=MaterialFunctionCall!(
    Sphere Radius=Constant() {r=1}
    Sphere Falloff=ScalarParameter() {defaultValue=2706.8237, parameterName=LightRadius}
    Unit Ray Direction=Normalize(VectorInput=@n21)
    Sphere Center=ObjectPositionWS()
    Ray Origin=@n20
   ) {materialFunction=/Game/Athena/Environments/MaterialFunctions/RayTraceSphereFalloffMinClamp}.Clamped Falloff)
  n23 = ScalarParameter() {defaultValue=2, parameterName=FalloffFog}
  n24 = Clamp(None=Add(A=Multiply(A=@n14, B=@n19), B=Power(Base=@n22, Exp=@n23)))
  n25 = Multiply(A=@n7, B=Multiply(A=VertexInterpolator(VS=@n11).PS, B=@n24))
  n26 = ScalarParameter() {defaultValue=2000, parameterName=DepthFadeDistance}
  n27 = FunctionInput(Preview=StaticBool()) {bUsePreviewValueAsDefault=true, inputType=FunctionInput_StaticBool}
  n28 = CollectionParameter() {
   collection=&FortniteMaterialParameters
   parameterName=FortniteActiveTimeOfDay
  }
  n29 = VectorParameter() {group=Emissive Autoexposure, parameterName=Emissive multiplier over day phases, defaultValue=<rgba 1 1 1 1>}
  n30 = DotProduct(A=@n28, B=AppendVector(A=@n29.RGB, B=@n29.RGB))
  n31 = Multiply(A=StaticSwitch(True=DepthFade(Opacity=@n25, FadeDistance=@n26), False=Multiply(A=@n7, B=@n24), Value=@n27), B=@n30)
  n32 = Subtract(A=CameraPositionWS(), B=WorldPosition().XYZ)
  n33 = Distance(A=ObjectPositionWS(), B=CameraPositionWS())
  n34 = ScalarParameter() {defaultValue=3579.1655, parameterName=SphereRadius}
  n35 = Ceil(None=Clamp(None=Subtract(A=@n33, B=@n34)))
  n36 = Add(A=Floor(None=@n10), B=OneMinus(None=@n35))
  n37 = Ceil(None=@n30)
  n38 = Add(A=StaticSwitch(True=Clamp(None=@n36), False=@n35, Value=@n27), B=OneMinus(None=Saturate(None=@n37)))
  n39 = ShadingPathSwitch(Default=Multiply(A=@n32, B=Saturate(None=@n38)), Mobile=@n32) {desc=Migrated from FeatureLevelSwitch}
  FunctionOutput(None=MakeMaterialAttributes!(
    BaseColor
    Metallic
    Specular
    Roughness
    Anisotropy
    EmissiveColor=Multiply(A=@n1, B=@n3)
    Opacity=Multiply(A=@n4.A, B=@n31)
    OpacityMask
    Normal
    Tangent
    WorldPositionOffset=ShadingPathSwitch(Default=@n39, Deferred=@n39, Forward=@n32, Mobile=@n32)
    SubsurfaceColor
    ClearCoat
    ClearCoatRoughness
    AmbientOcclusion
    Refraction
    CustomizedUVs_0
    CustomizedUVs_1
    CustomizedUVs_2
    CustomizedUVs_3
    CustomizedUVs_4
    CustomizedUVs_5
    CustomizedUVs_6
    CustomizedUVs_7
    PixelDepthOffset
    ShadingModel
    Displacement
   ))
  OneMinus(None=@n11)
 ]}
