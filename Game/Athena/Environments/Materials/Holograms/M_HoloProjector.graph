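# Additive, unlit, two-sided material for hologram projector beams, built on SplineThicken with tip and base widths scaled by object scale.
# Time-driven sines, randomized per Pivot Painter element, sweep beam vertices masked by red vertex color across a Left/HeightMax range, with optional bend and rotation.
# The swept position, converted to local space, becomes UVs for a poster texture that is tinted by Color.
# Camera-angle fade and a cylinder gradient are also present.
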Material /Game/Athena/Environments/Materials/Holograms/M_HoloProjector
graph {
 properties {blendMode=BLEND_Additive, pixelDepthOffsetMode=PDOM_Legacy, shadingModel=MSM_Unlit, twoSided=true, usedShadingModels=Unlit}
 nodes [
  n1 = ScalarParameter() {defaultValue=15.4776, parameterName=WidthTip}
  n2 = MaterialFunctionCall() {materialFunction=/Engine/Functions/Engine_MaterialFunctions02/WorldPositionOffset/ObjectScale}
  n3 = Max(A=@n2.Scale X, B=@n2.Scale X)
  n4 = ScalarParameter() {defaultValue=1, parameterName=Width}
  n5 = WorldPosition()
  n6 = ScalarParameter() {defaultValue=100, parameterName=Left}
  n7 = Time()
  n8 = MaterialFunctionCall() {materialFunction=/Engine/Functions/Engine_MaterialFunctions02/PivotPainter/PivotPainter_PerObjectData}
  n9 = Multiply(A=@n8.Random Value Per Element, B=ScalarParameter() {defaultValue=50.9562, parameterName=Random2})
  n10 = Sine(None=Multiply(A=Add(A=@n7, B=@n9), B=ScalarParameter() {defaultValue=0.03, parameterName=Speed2}))
  n11 = LinearInterpolate(A=@n6, B=Multiply(A=@n6) {constB=-1}, Alpha=Divide(A=Add(A=@n10)))
  n12 = Constant()
  n13 = ScalarParameter() {defaultValue=100, parameterName=HeightMax}
  n14 = Multiply(A=@n13) {constB=-1}
  n15 = Multiply(A=@n8.Random Value Per Element, B=ScalarParameter() {defaultValue=77.9935, parameterName=Random1})
  n16 = Multiply(A=Add(A=Time(), B=@n15), B=ScalarParameter() {defaultValue=0.03, parameterName=Speed})
  n17 = Divide(A=Add(A=Sine(None=@n16)))
  n18 = Transform(None=AppendVector(A=AppendVector(A=@n11, B=@n12), B=LinearInterpolate(A=@n13, B=@n14, Alpha=@n17))) {transformSourceType=TRANSFORMSOURCE_Local}
  n19 = Add(A=WorldPosition().XYZ, B=@n18)
  n20 = ObjectPositionWS()
  n21 = ComponentMask(None=Transform(None=Subtract(A=@n19, B=@n20)) {transformSourceType=TRANSFORMSOURCE_World, transformType=TRANSFORM_Local}) {r=true}
  n22 = RotateAboutAxis(
   NormalizedRotationAxis=ObjectOrientation()
   RotationAngle=Multiply(A=Divide(A=@n21) {constB=100}, B=ScalarParameter() {defaultValue=-0.0526, parameterName=Rotationamount})
   PivotPoint=@n20
   Position=@n19
  )
  n23 = Multiply(A=StaticSwitchParameter(True=Add(A=@n18, B=@n22), False=@n18) {defaultValue=true, parameterName=UseBend}, B=VertexColor().R)
  n24 = ObjectOrientation()
  n25 = Frac(None=Multiply(A=Time(), B=ScalarParameter() {defaultValue=0.1524, parameterName=Constant Rotation Speed}))
  n26 = ActorPositionWS()
  n27 = Add(A=@n23, B=WorldPosition().XYZ)
  n28 = StaticSwitchParameter(True=Add(A=@n23, B=RotateAboutAxis(NormalizedRotationAxis=@n24, RotationAngle=@n25, PivotPoint=@n26, Position=@n27)), False=@n23) {defaultValue=true, parameterName=UseRotation}
  Add(
   A=MaterialFunctionCall(WidthBase=Multiply(A=@n1, B=@n3), WidthTip=Multiply(A=@n3, B=@n4), WorldPosition=Add(A=@n5.XYZ, B=@n28)) {materialFunction=&SplineThicken}.WorldPositionOffset
   B=@n28
  )
  Add(A=@n19, B=Constant3Vector() {constant=<rgba 100 0 0 0>})
  Multiply(A=@n15, B=ScalarParameter() {defaultValue=7.8181, parameterName=test2})
  n29 = Transform(None=Constant3Vector() {constant=<rgba 0 1 0 0>})
  n30 = Power(Base=DotProduct(A=CameraVectorWS(), B=@n29), Exp=ScalarParameter() {defaultValue=0.6, parameterName=Frontfade})
  Multiply(B=OneMinus(None=Clamp(None=@n30)))
  n31 = VectorParameter() {parameterName=Color, defaultValue=<rgba 1 1 1 1>}
  n32 = Add(A=WorldPosition() {worldPositionShaderOffset=WPT_ExcludeAllShaderOffsets}.XYZ, B=@n18)
  n33 = ObjectPositionWS()
  n34 = ComponentMask(None=Transform(None=Subtract(A=@n32, B=@n33)) {transformSourceType=TRANSFORMSOURCE_World, transformType=TRANSFORM_Local}) {b=true, r=true}
  n35 = Multiply(A=AppendVector(A=@n6, B=@n13), B=Constant2Vector() {g=-2, r=2}) {constB=2}
  n36 = TextureSampleParameter2D(UVs=Subtract(A=Divide(A=@n34, B=@n35), B=Constant2Vector() {g=0.5, r=0.5}) {constB=0.5}) {parameterName=BaseColor, texture=/Game/Packages/DS_Fortnight/SM/Textures/T_FORT_RiskyReelsPoster}
  Multiply(A=Multiply(A=@n31.RGB, B=@n36.RGB), B=ScalarParameter() {defaultValue=0.37, parameterName=Brightness})
  Multiply(A=Multiply(A=@n31.RGB, B=Constant() {r=2}), B=Clamp(None=@n36.RGB) {clampMode=CMODE_ClampMin, minDefault=0.25})
  Power(Exp=ScalarParameter() {defaultValue=1.5, parameterName=Contrast})
  ScalarParameter() {parameterName=OffsetProjectionZ}
  ScalarParameter() {defaultValue=3, parameterName=Glow}
  StaticSwitchParameter(True=Normalize(VectorInput=@n18), False=@n29) {defaultValue=true, parameterName=UseRotation}
  TextureSample(UVs=MaterialFunctionCall(XY=TextureCoordinate()) {materialFunction=&Swizzle}.YX) {texture=/Game/Global/Textures/Default/Masks_and_Gradients/CylGradient}
 ]
}
