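# MF_AnimatedTexture is a material function that composes animated advertisement imagery for holograms: a patterned background with two animated textures layered on top.
# The tinted background uses scale, skew, pan and rotation, with switchable blend modes.
# Each texture lerps offset and scale and rotates on sine or sawtooth time, with optional wait periods and fade in/out.
# Channel-selected masks composite both over the background; Texture 2 adds a time offset and flat-color option.
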MaterialFunction /Game/Athena/Environments/Materials/Holograms/MF_AnimatedTexture
graph {nodes [
  n1 = ScalarParameter() {group=Background, parameterName=BackGroundOffsetX}
  n2 = ScalarParameter() {group=Background, parameterName=BackGroundOffsetY}
  Add(A=@n1, B=@n2)
  n3 = VectorParameter() {group=Background, parameterName=BackGroundColor2, defaultValue=<rgba 0 0 0 1>}
  n4 = AppendVector(A=@n2, B=@n1)
  n5 = Multiply(
   A=TextureCoordinate()
   B=AppendVector(
    A=ScalarParameter() {defaultValue=2.1724, group=Advertisment Scale Proportions, parameterName=UV_Xscale}
    B=ScalarParameter() {defaultValue=1, group=Advertisment Scale Proportions, parameterName=UV_Yscale}
   )
  )
  n6 = ScalarParameter() {defaultValue=1, group=Background, parameterName=BackGroundScaleX}
  n7 = ScalarParameter() {defaultValue=1, group=Background, parameterName=BackGroundScaleY}
  n8 = LinearInterpolate(
   A=Constant2Vector() {g=1, r=1}
   B=ComponentMask(None=@n5) {g=true, r=true}
   Alpha=ScalarParameter() {group=Background, parameterName=BackgroundUVSkew}
  )
  n9 = Multiply(A=Multiply(A=@n5, B=AppendVector(A=@n6, B=@n7)), B=@n8)
  n10 = Panner(
   Coordinate=@n9
   Speed=AppendVector(A=ScalarParameter() {group=Background, parameterName=BackGroundPanX}, B=ScalarParameter() {group=Background, parameterName=BackGroundPanY})
  )
  n11 = ScalarParameter() {group=Background, parameterName=BackGroundRotation}
  n12 = TextureSampleParameter2D(UVs=Add(A=@n4, B=Rotator(Coordinate=@n10, Time=@n11))) {parameterName=BackGoundPattern, samplerType=SAMPLERTYPE_LinearColor, texture=/Game/Athena/SafeZone/Lines}
  n13 = Multiply(A=@n3.RGB, B=@n12.RGB)
  n14 = Multiply(A=@n12.R, B=@n3.RGB)
  n15 = VectorParameter() {group=Background, parameterName=BackGroundColor, defaultValue=<rgba 1 1 1 1>}
  n16 = AppendVector(
   A=ScalarParameter() {group=Background, parameterName=BackGround2OffsetY}
   B=ScalarParameter() {group=Background, parameterName=BackGround2OffsetX}
  )
  n17 = ScalarParameter() {defaultValue=1, group=Background, parameterName=BackGround2ScaleX}
  n18 = ScalarParameter() {defaultValue=1, group=Background, parameterName=BackGround2ScaleY}
  n19 = Multiply(A=Multiply(A=@n5, B=AppendVector(A=@n17, B=@n18)), B=@n8)
  n20 = TextureSampleParameter2D(UVs=StaticSwitchParameter(True=Add(A=@n16, B=@n19), False=Add(A=@n9, B=@n4)) {group=Background, parameterName=BackGround2uniqueUVs}) {parameterName=BackGoundPattern2, samplerType=SAMPLERTYPE_LinearColor, texture=&Lines}
  n21 = StaticSwitchParameter(
   True=LinearInterpolate(
    A=StaticSwitchParameter(True=@n13, False=@n14) {group=Background, parameterName=BackGroundLerporMult}
    B=Multiply(A=@n15.RGB, B=@n20.RGB)
    Alpha=@n20.A
   )
   False=StaticSwitchParameter(True=LinearInterpolate(A=@n15.RGB, B=@n3.RGB, Alpha=@n12.R), False=Multiply(A=@n15.RGB, B=@n12.RGB)) {group=Background, parameterName=BackGroundLerporMult}
  ) {group=Background, parameterName=BackgroundUseAlphaLerp}
  n22 = ScalarParameter() {group=Texture1, parameterName=OffsetXUVTex1_a}
  n23 = ScalarParameter() {group=Texture1, parameterName=OffsetYUVTex1_a}
  n24 = ScalarParameter() {group=Texture1, parameterName=OffsetXUVTex1_b}
  n25 = ScalarParameter() {group=Texture1, parameterName=OffsetYUVTex1_b}
  n26 = Time()
  n27 = Multiply(A=@n26, B=ScalarParameter() {defaultValue=0.46, group=Texture1, parameterName=Speed_tex1})
  n28 = ScalarParameter() {defaultValue=0.4898, group=Texture1, parameterName=WaitTime1}
  n29 = Subtract(A=Frac(None=@n27), B=@n28)
  n30 = Subtract(A=Constant() {r=1}, B=@n28)
  n31 = StaticSwitchParameter(True=Clamp(None=Divide(A=@n29, B=@n30)), False=Frac(None=@n27)) {group=Texture1, parameterName=UseWaitTimeTexture1}
  n32 = StaticSwitchParameter(True=Sine(None=@n27), False=@n31) {group=Texture1, parameterName=MovementSineOrFracTex1}
  n33 = Add(A=@n5, B=LinearInterpolate(A=AppendVector(A=@n22, B=@n23), B=AppendVector(A=@n24, B=@n25), Alpha=@n32))
  n34 = Multiply(A=@n26, B=ScalarParameter() {defaultValue=0.46, group=Texture1, parameterName=Speed_Rotation_tex1})
  n35 = ScalarParameter() {defaultValue=0.4898, group=Texture1, parameterName=WaitTime1}
  n36 = Subtract(A=Frac(None=@n34), B=@n35)
  n37 = Subtract(A=Constant() {r=1}, B=@n35)
  n38 = StaticSwitchParameter(True=Clamp(None=Divide(A=@n36, B=@n37)), False=Frac(None=@n34)) {group=Texture1, parameterName=UseWaitTimeTexture1}
  n39 = Multiply(
   A=ScalarParameter() {group=Texture1, parameterName=Rotation_tex1}
   B=StaticSwitchParameter(True=Sine(None=@n34), False=@n38) {group=Texture1, parameterName=RotationSineOrFracTex1}
  )
  n40 = LinearInterpolate(
   A=ScalarParameter() {defaultValue=1, group=Texture1, parameterName=ScaleUVTex1_a}
   B=ScalarParameter() {defaultValue=1, group=Texture1, parameterName=ScaleUVTex1_b}
   Alpha=@n32
  )
  n41 = Multiply(A=@n40, B=ScalarParameter() {defaultValue=1, group=Texture1, parameterName=ScaleUVTex1_Y})
  n42 = TextureSampleParameter2D(UVs=MaterialFunctionCall(UVs=Rotator(Coordinate=@n33, Time=@n39), Texture Scale=AppendVector(A=@n40, B=@n41)) {materialFunction=&ScaleUVsByCenter}.UVs) {group=Texture1, parameterName=AnimatedTexture1, texture=/Game/Athena/Environments/Textures/T_Fish_Adv_1}
  n43 = Multiply(A=VectorParameter() {group=Texture1, parameterName=ColorTexture1, defaultValue=<rgba 1 1 1 1>}.RGB, B=@n42.RGB)
  n44 = ChannelMaskParameter(None=@n42.RGBA) {group=Texture1, maskChannel=Alpha, parameterName=Texture1MaskChannel, defaultValue=<rgba 0 0 0 1>}
  n45 = StaticSwitchParameter(True=@n31, False=@n38) {group=Texture1, parameterName=FadeOnMoveOrRotateTex1}
  n46 = ScalarParameter() {defaultValue=0.01, group=Texture1, parameterName=FadeInOutPauseTex1}
  n47 = ScalarParameter() {defaultValue=3, group=Texture1, parameterName=FadeInOutTex1}
  n48 = Saturate(None=Divide(A=Subtract(A=@n45, B=@n46), B=@n47))
  n49 = OneMinus(None=@n45)
  n50 = Saturate(None=Divide(A=Subtract(A=@n49, B=@n46), B=@n47))
  n51 = StaticSwitchParameter(True=Multiply(A=@n44.R, B=Multiply(A=@n48, B=@n50) {constB=2}), False=@n44.R) {group=Texture1, parameterName=FadeTexture1}
  n52 = VectorParameter() {group=Texture2, parameterName=ColorTexture2, defaultValue=<rgba 1 1 1 1>}
  n53 = ScalarParameter() {group=Texture2, parameterName=OffsetXUVTex2_a}
  n54 = ScalarParameter() {group=Texture2, parameterName=OffsetYUVTex2_a}
  n55 = ScalarParameter() {group=Texture2, parameterName=OffsetXUVTex2_b}
  n56 = ScalarParameter() {group=Texture2, parameterName=OffsetYUVTex2_b}
  n57 = Add(A=Time(), B=ScalarParameter() {group=Texture2, parameterName=Time_Offset_tex2})
  n58 = Multiply(A=@n57, B=ScalarParameter() {defaultValue=0.46, group=Texture2, parameterName=Speed_tex2})
  n59 = ScalarParameter() {defaultValue=0.4898, group=Texture2, parameterName=WaitTime2}
  n60 = Subtract(A=Frac(None=@n58), B=@n59)
  n61 = Subtract(A=Constant() {r=1}, B=@n59)
  n62 = StaticSwitchParameter(True=Clamp(None=Divide(A=@n60, B=@n61)), False=Frac(None=@n58)) {group=Texture2, parameterName=UseWaitTimeTexture2}
  n63 = StaticSwitchParameter(True=Sine(None=@n58), False=@n62) {group=Texture2, parameterName=MovementSineOrFracTex2}
  n64 = Add(A=@n5, B=LinearInterpolate(A=AppendVector(A=@n53, B=@n54), B=AppendVector(A=@n55, B=@n56), Alpha=@n63))
  n65 = Multiply(A=@n57, B=ScalarParameter() {group=Texture2, parameterName=Speed_Rotation_tex2})
  n66 = ScalarParameter() {defaultValue=0.4898, group=Texture2, parameterName=WaitTime2}
  n67 = Subtract(A=Frac(None=@n65), B=@n66)
  n68 = Subtract(A=Constant() {r=1}, B=@n66)
  n69 = StaticSwitchParameter(True=Clamp(None=Divide(A=@n67, B=@n68)), False=Frac(None=@n65)) {group=Texture2, parameterName=UseWaitTimeTexture2}
  n70 = Multiply(
   A=ScalarParameter() {group=Texture2, parameterName=Rotation_tex2}
   B=StaticSwitchParameter(True=Sine(None=@n65), False=@n69) {group=Texture2, parameterName=RotationSineOrFracTex2}
  )
  n71 = LinearInterpolate(
   A=ScalarParameter() {defaultValue=1, group=Texture2, parameterName=ScaleUVTex2_a}
   B=ScalarParameter() {defaultValue=1, group=Texture2, parameterName=ScaleUVTex2_b}
   Alpha=@n63
  )
  n72 = Multiply(A=@n71, B=ScalarParameter() {defaultValue=1, group=Texture2, parameterName=ScaleUVTex2_Y})
  n73 = TextureSampleParameter2D(UVs=MaterialFunctionCall(UVs=Rotator(Coordinate=@n64, Time=@n70), Texture Scale=AppendVector(A=@n71, B=@n72)) {materialFunction=&ScaleUVsByCenter}.UVs) {group=Texture2, parameterName=AnimatedTexture2, samplerType=SAMPLERTYPE_LinearColor, texture=/Game/Athena/Environments/Textures/T_Fish_Adv_2}
  n74 = Multiply(A=@n52.RGB, B=@n73.RGB)
  n75 = ChannelMaskParameter(None=@n73.RGBA) {group=Texture2, parameterName=Texture2MaskChannel}
  n76 = StaticSwitchParameter(True=@n62, False=@n69) {group=Texture2, parameterName=FadeOnMoveOrRotateTex2}
  n77 = ScalarParameter() {defaultValue=0.01, group=Texture2, parameterName=FadeInOutPauseTex2}
  n78 = ScalarParameter() {defaultValue=3, group=Texture2, parameterName=FadeInOutTex2}
  n79 = Divide(A=Subtract(A=@n76, B=@n77), B=@n78)
  n80 = Divide(A=Subtract(A=OneMinus(None=@n76), B=@n77), B=@n78)
  n81 = Multiply(A=@n75.R, B=Multiply(A=Saturate(None=@n79), B=Saturate(None=@n80)) {constB=2})
  FunctionOutput(None=LinearInterpolate(
    A=LinearInterpolate(A=@n21, B=@n43, Alpha=@n51)
    B=StaticSwitchParameter(True=@n52.RGB, False=@n74) {group=Texture2, parameterName=Texture2DirectColor}
    Alpha=StaticSwitchParameter(True=@n81, False=@n75.R) {group=Texture2, parameterName=FadeTexture2}
   ))
  Sine(None=Divide(A=@n45))
 ]}
