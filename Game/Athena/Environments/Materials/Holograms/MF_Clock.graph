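# Material function that renders a digital clock mask, likely for holograms, showing in-game hours and minutes from the TimeOfDay_0-1 collection parameter.
# Center-scaled UVs are split into up to eight cells (five by default), with nested Ifs choosing each cell's time digit or parameter glyph.
# The chosen index offsets into a horizontal digit-strip texture, sampled with original-UV derivatives for mip selection and masked to the padded area.
# A bias parameter offsets rounding error in digit extraction.
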MaterialFunction /Game/Athena/Environments/Materials/Holograms/MF_Clock
graph {nodes [
  CollectionParameter() {collection=&FortniteMaterialParameters, parameterName=RealGameTime}
  n1 = FunctionInput(Preview=TextureCoordinate()) {bUsePreviewValueAsDefault=true, inputName=UVs, inputType=FunctionInput_Vector2, sortPriority=1}
  n2 = ScalarParameter() {defaultValue=0.9128, group=Numbers, parameterName=u_NumberPaddingScale}
  n3 = ScalarParameter() {defaultValue=0.3143, group=Numbers, parameterName=v_NumberPaddingScale}
  n4 = MaterialFunctionCall(
   UVs=@n1
   Texture Scale=FunctionInput(Preview=AppendVector(A=@n2, B=@n3)) {bUsePreviewValueAsDefault=true, inputName=UVPadding, inputType=FunctionInput_Vector2, sortPriority=1}
  ) {materialFunction=&ScaleUVsByCenter}
  n5 = FunctionInput(Preview=Constant() {desc=Number of digits, r=5}) {
   bUsePreviewValueAsDefault=true
   desc=Only whole numbers between 1 and 8 Using a constant saves some instructions
   description=Using a constant saves some instructions
   inputName="NumberofDigits(int)"
   inputType=FunctionInput_Scalar
   sortPriority=2
  }
  n6 = Multiply(A=ComponentMask(None=@n4.UVs) {r=true}, B=@n5) {constB=5}
  n7 = Constant() {r=1}
  n8 = Constant() {r=4}
  n9 = Constant() {r=7}
  n10 = ComponentMask(None=@n6) {r=true}
  n11 = Multiply(A=ComponentMask(None=CollectionParameter() {collection=&FortniteMaterialParameters, parameterName=TimeOfDay_0-1}) {r=true}) {constB=24, desc=Hours}
  n12 = Divide(A=Floor(None=@n11) {desc=CurrentHours}) {constB=10}
  n13 = Floor(None=@n12) {desc=firstDigit}
  n14 = ScalarParameter() {defaultValue=0.0952, parameterName=RoundingErrorBias}
  n15 = Add(A=Frac(None=@n12), B=@n14) {constB=0.5}
  n16 = If(A=Constant() {r=1}, B=@n10, "A > B"=@n13, "A < B"=Floor(None=Multiply(A=@n15) {constB=10}) {desc=2nd digit})
  n17 = If(A=Constant() {r=2}, B=@n10, "A > B"=@n16, "A < B"=ScalarParameter() {defaultValue=-1, group=Numbers, parameterName=Digit_3})
  n18 = Frac(None=@n11)
  n19 = Divide(A=Floor(None=Multiply(A=@n18) {constB=60, desc=MinCurrent})) {constB=10}
  n20 = If(A=Constant() {r=3}, B=@n10, "A > B"=@n17, "A < B"=Floor(None=@n19) {desc=3rd digit})
  n21 = Add(A=Frac(None=@n19), B=@n14)
  n22 = If(A=Constant() {r=4}, B=@n10, "A > B"=@n20, "A < B"=Floor(None=Multiply(A=@n21) {constB=10}) {desc=4th digit})
  n23 = If(A=Constant() {r=5}, B=@n10, "A > B"=@n22, "A < B"=ScalarParameter() {defaultValue=-1, group=Numbers, parameterName=Digit_6})
  n24 = If(A=Constant() {r=6}, B=@n10, "A > B"=@n23, "A < B"=ScalarParameter() {defaultValue=7, group=Numbers, parameterName=Digit_7})
  n25 = If(A=Constant() {r=7}, B=@n10, "A > B"=@n24, "A < B"=ScalarParameter() {defaultValue=8, group=Numbers, parameterName=Digit_8})
  n26 = If(A=@n5, B=Constant() {r=8}, "A > B"=@n25, "A == B"=@n25, "A < B"=@n25)
  n27 = If(
   A=@n5
   B=Constant() {r=5}
   "A > B"=If(A=@n5, B=Constant() {r=6}, "A > B"=If(A=@n5, B=@n9, "A > B"=@n26, "A == B"=@n24, "A < B"=@n24), "A == B"=@n23, "A < B"=@n23)
   "A == B"=@n22
   "A < B"=@n22
  )
  n28 = If(
   A=@n5
   B=Constant() {r=2}
   "A > B"=If(A=@n5, B=Constant() {r=3}, "A > B"=If(A=@n5, B=@n8, "A > B"=@n27, "A == B"=@n20, "A < B"=@n20), "A == B"=@n17, "A < B"=@n17)
   "A == B"=@n16
   "A < B"=@n16
  )
  n29 = Add(A=Frac(None=@n6), B=Floor(None=If(A=@n5, B=@n7, "A > B"=@n28, "A == B"=@n13, "A < B"=@n13)))
  n30 = ScalarParameter() {defaultValue=11, group=Numbers, parameterName=Numbers of Digits On Texture}
  n31 = TextureSampleParameter2D!(
   UVs=AppendVector(A=Divide(A=@n29, B=@n30) {constB=11}, B=ComponentMask(None=@n4.UVs) {g=true})
   "DDX(UVs)"=DDX(Value=@n1)
   "DDY(UVs)"=DDY(Value=@n1)
   Apply View MipBias
  ) {
   group=Numbers
   mipValueMode=TMVM_Derivative
   parameterName=NumbersTexture
   samplerType=SAMPLERTYPE_Masks
   texture=/Game/Athena/Playlists/Barrier/Textures/T_Numbers2_M
  }
  FunctionOutput(None=Multiply(A=@n31.R, B=@n4.0-1 mask))
  ScalarParameter() {defaultValue=4, group=Numbers, parameterName=Digit_4}
  ScalarParameter() {defaultValue=1, group=Numbers, parameterName=Digit_1}
 ]}
