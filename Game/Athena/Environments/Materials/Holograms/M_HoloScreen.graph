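# Two-sided, unlit masked material for holographic screens, showing a poster, animated texture or clock through static switches.
# Position-seeded time drives the poster's sine-scrolled scanlines and periodic parallax-offset ghosting, plus digital-noise blocks over all content, before time-of-day glow and vignette tint.
# Opacity mask blends temporal dither with edge falloff using near-camera fade.
# World position offset bends and spins the mesh or uses party bus animation, zeroed for HLOD and proxies.
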Material /Game/Athena/Environments/Materials/Holograms/M_HoloScreen
graph {
 properties {
  bUseMaterialAttributes=true
  blendMode=BLEND_Masked
  pixelDepthOffsetMode=PDOM_Legacy
  shadingModel=MSM_Unlit
  twoSided=true
  usedShadingModels=Unlit
 }
 nodes [
  n1 = Custom!(None) {
   code=View.TemporalAAParams.xy
   desc="\n*** Original source before expression upgrade ***\nView.TemporalAAParams\n*** Original source before expression upgrade ***\nFrame.TemporalAAParams"
   description=TemporalAA SampleID
   outputType=CMOT_Float2
  }
  MaterialExpressionAdd_2 = Add(
   A=MaterialFunctionCall!(TextureResolution=Constant2Vector() {g=1, r=1}) {materialFunction=/Engine/Functions/Engine_MaterialFunctions02/Texturing/ScreenAlignedPixelToPixelUVs}.Result
   B=ComponentMask(None=@n1) {r=true}
  )
  n2 = Custom!(p=@MaterialExpressionAdd_2) {
   code="Mod( ((uint)(p.x) + 2 * (uint)(p.y)) , 5 )"
   outputType=CMOT_Float1
   inputs [{inputName=p, input {expression=&self:MaterialExpressionAdd_2}}]
  }
  n3 = MaterialFunctionCall(TextureResolution=Constant2Vector() {g=64, r=64}) {materialFunction=&ScreenAlignedPixelToPixelUVs}
  Add(
   A=@n2
   B=Multiply(A=TextureSample(UVs=@n3.Result) {samplerType=SAMPLERTYPE_LinearColor, texture=/Game/Global/Textures/Default/Noise/T_Atlas_DigitalNoise_01}.R)
  ) {constB=0}
  Constant3Vector() {constant=<rgba 0 0 1 0>}
  Constant() {r=1}
  Constant() {r=1}
  Divide(A=@n2) {constB=6}
  n4 = MaterialFunctionCall() {materialFunction=/Engine/Functions/Engine_MaterialFunctions02/WorldPositionOffset/ObjectScale}
  n5 = Multiply(
   A=TextureCoordinate()
   B=Multiply(A=ScalarParameter() {defaultValue=0.5619, parameterName=BlockTiling}, B=AppendVector(A=@n4.Scale X, B=@n4.Scale X))
  )
  n6 = ObjectPositionWS()
  n7 = Divide(A=Divide(A=ComponentMask(None=@n6) {r=true}, B=ComponentMask(None=@n6) {g=true}), B=ComponentMask(None=@n6) {b=true})
  n8 = Add(A=Time(), B=Multiply(A=Frac(None=@n7), B=ScalarParameter() {defaultValue=23425.4395, parameterName=Random}))
  n9 = Multiply(A=@n8, B=ScalarParameter() {defaultValue=12.0689, parameterName=Block Speed A})
  n10 = TextureSample(UVs=Panner(Coordinate=@n5, Time=Ceil(None=@n9)) {bFractionalPart=true, speedX=-0.477, speedY=-0.334}) {samplerType=SAMPLERTYPE_LinearColor, texture=&T_Atlas_DigitalNoise_01}
  LinearInterpolate(
   A=ScalarParameter() {defaultValue=0.05, parameterName=Blocks Dark}
   B=ScalarParameter() {defaultValue=1, parameterName=Blocks Light}
   Alpha=@n10.R
  )
  MaterialFunctionCall() {materialFunction=&DitherTemporalAA}
  MaterialFunctionCall() {materialFunction=&DitherTemporalAA}
  n11 = VectorParameter() {parameterName=wiggle1, defaultValue=<rgba 0.2 0.4 0.3524 1>}
  n12 = VectorParameter() {parameterName=wiggle1_1, defaultValue=<rgba 0 1 0 1>}
  MaterialFunctionCall!(
   Number of Scanlines=ScalarParameter() {defaultValue=150, parameterName=Scanline Amount}
   Scanline Speed
   "Scanlines Thickness (Bias)"=ScalarParameter() {defaultValue=2, parameterName=ScanlineSize}
   Scanlines Dark Value=ScalarParameter() {defaultValue=0.5, parameterName=Scanlines Dark}
   Scanlines Light Value=ScalarParameter() {defaultValue=1, parameterName=Scanlines Light}
   Block Opacity=StaticSwitchParameter(
    True=MaterialFunctionCall!(WPO Intensity A=@n11.RGB, WPO Intensity B=@n12.RGB, Time, Bias) {materialFunction=/Game/Effects/Fort_Effects/Effects/Atlas/Materials/MF_Atlas_HologramWiggle_01}.Intense 0-1
    False=Constant()
   ) {defaultValue=true, parameterName=Use Blocks?}
   UVs
   Time
   Block Repeat
  ) {materialFunction=/Game/Effects/Fort_Effects/Effects/Atlas/Materials/MF_Atlas_Scanlines_01}
  n13 = TextureCoordinate()
  n14 = ScalarParameter() {defaultValue=4, parameterName=ScanLineSize1}
  n15 = Sine(None=Multiply(A=@n8, B=ScalarParameter() {defaultValue=0.1, parameterName=ScanLineSpeed1}))
  n16 = ScalarParameter() {defaultValue=1, parameterName=ScanLineMovementamount1}
  n17 = TextureSample(UVs=Add(A=Multiply(A=@n13, B=@n14), B=Multiply(A=@n15, B=@n16))) {samplerType=SAMPLERTYPE_Masks, texture=/Game/Weapons/WeaponSkins/Textures/T_SimpleLine}
  n18 = TextureCoordinate()
  n19 = ScalarParameter() {defaultValue=8, parameterName=ScanLineSize2}
  n20 = Sine(None=Multiply(A=@n8, B=ScalarParameter() {defaultValue=-0.133, parameterName=ScanLineSpeed2}))
  n21 = ScalarParameter() {defaultValue=1, parameterName=ScanLineMovementamount2}
  n22 = TextureSample(UVs=Add(A=Multiply(A=@n18, B=@n19), B=Multiply(A=@n20, B=@n21))) {samplerType=SAMPLERTYPE_Masks, texture=&T_SimpleLine}
  n23 = TextureCoordinate()
  n24 = Constant() {r=1}
  n25 = Frac(None=Multiply(A=@n8, B=ScalarParameter() {defaultValue=0.2, parameterName=TimeSpeed}))
  n26 = ScalarParameter() {defaultValue=16, parameterName=WaitTime}
  n27 = Floor(None=Clamp(None=Multiply(A=@n25, B=@n26)))
  n28 = LinearInterpolate(
   A=ScalarParameter() {defaultValue=-0.04, parameterName=BumpOffset}
   B=ScalarParameter() {defaultValue=0.005, parameterName=BumpOffset2}
   Alpha=@n27
  )
  n29 = Add(
   A=TextureSampleParameter2D(UVs=@n23) {parameterName=BaseColor, texture=/Game/Packages/DS_Fortnight/SM/Textures/T_FORT_RiskyReelsPoster}.RGB
   B=TextureSampleParameter2D(UVs=BumpOffset(Coordinate=@n23, Height=@n24, HeightRatioInput=@n28)) {parameterName=BaseColor, texture=&T_FORT_RiskyReelsPoster}.RGB
  )
  n30 = Multiply(
   A=Multiply(A=Add(A=@n17.R, B=@n22.R), B=@n29)
   B=VectorParameter() {parameterName=ScanlineColor, defaultValue=<rgba 0.3802 0.3802 0.3802 1>}.RGB
  )
  n31 = Multiply(A=VectorParameter() {parameterName=Color, defaultValue=<rgba 1 1 1 1>}.RGB, B=Divide(A=@n29))
  n32 = StaticSwitchParameter(
   True=Multiply(
    A=VectorParameter() {parameterName=ClockColor, defaultValue=<rgba 1 1 1 1>}.RGB
    B=MaterialFunctionCall!(UVs, UVPadding, "NumberofDigits(int)") {materialFunction=&/MF_Clock}.Result
   )
   False=StaticSwitchParameter(
    True=MaterialFunctionCall() {materialFunction=&/MF_AnimatedTexture}.Result
    False=Add(A=@n30, B=@n31)
   ) {parameterName=UseAnimatedTextures}
  ) {parameterName=UseClock}
  n33 = ScalarParameter() {defaultValue=0.125, parameterName=LageBlockMask}
  n34 = Ceil(None=Multiply(A=@n8, B=ScalarParameter() {defaultValue=1, parameterName=SpeedWaittilSwitch}))
  n35 = Constant() {r=0.4563}
  n36 = TextureSample(UVs=Add(A=Multiply(A=@n5, B=@n33), B=Multiply(A=@n34, B=@n35))) {samplerType=SAMPLERTYPE_LinearColor, texture=&T_Atlas_DigitalNoise_01}
  n37 = Constant()
  n38 = LinearInterpolate(
   B=@n10.R
   Alpha=Multiply(A=LinearInterpolate(A=@n36.R, B=@n37, Alpha=@n27), B=ScalarParameter() {defaultValue=0.2, parameterName=BlockOpacity})
  ) {constA=1}
  n39 = ScalarParameter() {defaultValue=1, parameterName=Glow_Multi_Morning, sortPriority=33}
  n40 = ScalarParameter() {defaultValue=1, parameterName=Glow_Multi_Day, sortPriority=33}
  n41 = ScalarParameter() {defaultValue=1, parameterName=Glow_Multi_Evening, sortPriority=33}
  n42 = ScalarParameter() {defaultValue=1, parameterName=Glow_Multi_Night, sortPriority=33}
  n43 = Multiply(
   A=Power(Base=Multiply(A=@n38, B=@n32), Exp=ScalarParameter() {defaultValue=1.5, parameterName=Contrast})
   B=Multiply(
    A=ScalarParameter() {defaultValue=3, parameterName=Glow}
    B=MaterialFunctionCall(Value_Morning=@n39, Value_Day=@n40, Value_Evening=@n41, Value_Night=@n42) {materialFunction=&Scalar_Time_of_Day}.Result
   )
  )
  n44 = VectorParameter() {parameterName=ColorVignette, defaultValue=<rgba 0.3523 0 0.4115 1>}
  n45 = Distance(A=TextureCoordinate(), B=Constant2Vector() {g=0.5, r=0.5})
  n46 = ScalarParameter() {defaultValue=0.22, parameterName=VigneteOffset}
  n47 = Saturate(None=Divide(A=Subtract(A=@n45, B=@n46), B=ScalarParameter() {defaultValue=1.0615, parameterName=Vignete}))
  n48 = LinearInterpolate(A=VectorParameter() {parameterName=ColorVigneteCenter, defaultValue=<rgba 1 1 1 1>}.RGB, B=@n44.RGB, Alpha=@n47)
  n49 = Add(A=@n43, B=@n48)
  n50 = LinearInterpolate(A=@n43, B=@n44.RGB, Alpha=@n47)
  n51 = TextureCoordinate()
  n52 = ComponentMask(None=@n51) {r=true}
  n53 = Multiply(A=@n52, B=OneMinus(None=@n52))
  n54 = ComponentMask(None=@n51) {g=true}
  n55 = Multiply(A=@n54, B=OneMinus(None=@n54))
  n56 = Power(
   Base=Multiply(A=Multiply(A=@n53, B=@n55), B=ScalarParameter() {defaultValue=4, parameterName=MaskEdgeFalloffMult}) {constB=4}
   Exp=ScalarParameter() {defaultValue=0.1762, parameterName=MaskEdgeFalloff}
  )
  n57 = ScalarParameter() {defaultValue=100, parameterName=NearFadeDistance}
  n58 = LinearInterpolate(
   A=MaterialFunctionCall() {materialFunction=&DitherTemporalAA}.Result
   B=@n56
   Alpha=Multiply(
    A=ScalarParameter() {defaultValue=0.781, parameterName=Opacity}
    B=MaterialFunctionCall(Distance=@n57) {materialFunction=&Near_Camera_Fade}.Result
   )
  )
  n59 = Constant3Vector()
  n60 = WorldPosition()
  n61 = ObjectPositionWS()
  n62 = Subtract(A=@n60.XYZ, B=@n61)
  n63 = Divide(A=ComponentMask(None=Transform(None=@n62) {transformSourceType=TRANSFORMSOURCE_World, transformType=TRANSFORM_Local}) {r=true}) {constB=100}
  n64 = ScalarParameter() {defaultValue=-0.0526, parameterName=Rotationamount}
  n65 = StaticSwitchParameter(
   True=RotateAboutAxis(NormalizedRotationAxis=ObjectOrientation(), RotationAngle=Multiply(A=@n63, B=@n64), PivotPoint=@n61, Position=@n60.XYZ)
   False=Constant3Vector()
  ) {defaultValue=true, parameterName=UseBend}
  n66 = Time()
  n67 = ScalarParameter() {defaultValue=0.1524, parameterName=Constant Rotation Speed}
  n68 = RotateAboutAxis(
   NormalizedRotationAxis=ObjectOrientation()
   RotationAngle=Frac(None=Multiply(A=@n66, B=@n67))
   PivotPoint=ActorPositionWS()
   Position=Add(A=@n65, B=WorldPosition().XYZ)
  )
  n69 = MaterialProxyReplace(Realtime=StaticSwitchParameter(True=Add(A=@n65, B=@n68), False=@n65) {defaultValue=true, parameterName=UseRotation}, MaterialProxy=@n59)
  n70 = SetMaterialAttributes!(
   MaterialAttributes
   Emissive Color=MaterialFunctionCall(Emissive=StaticSwitchParameter(True=@n49, False=@n50) {parameterName=UseVignetteLerpAdd}) {materialFunction=&MF_FX_EmissiveAdjustments}.Result
   Opacity=@n58
   Opacity Mask=MaterialProxyReplace(Realtime=@n58, MaterialProxy=Constant() {r=1})
   World Position Offset=StaticSwitchParameter(
    True=MaterialFunctionCall() {materialFunction=/Game/Athena/Aircraft/PartyBus_RT/Materials/MF_PartyBusAnimationWPO}.Result
    False=StaticSwitchParameter(True=@n59, False=@n69) {parameterName=HLOD}
   ) {parameterName=UsePartyBusAnimation}
  ) {attributeSetTypes [
    #1
    #2
    #3
    #4
   ]}
  MaterialFunctionCall(MA In=@n70) {materialFunction=&MF_CreativeMoveTool}
  Multiply(A=@n48, B=@n43)
  n71 = TwoSidedSign()
  Multiply(
   A=@n38
   B=LinearInterpolate(
    A=ScalarParameter() {defaultValue=0.7, parameterName=OpcaityBackFace}
    B=ScalarParameter() {defaultValue=0.7524, parameterName=Opcaity}
    Alpha=Saturate(None=@n71)
   )
  )
  Multiply(A=@n8)
  Multiply(A=@n36.R, B=@n27)
  n72 = Multiply(A=TextureCoordinate(), B=ScalarParameter() {defaultValue=2, parameterName=HoloScreens})
  n73 = ScalarParameter() {defaultValue=0.05, parameterName=MaskPanSpeed}
  n74 = Multiply(A=@n73) {constB=-1}
  Multiply(
   A=TextureSample(UVs=Panner(Coordinate=@n72, Speed=@n73)) {samplerType=SAMPLERTYPE_LinearColor, texture=&T_Atlas_DigitalNoise_01}.R
   B=TextureSample(UVs=Panner(Coordinate=@n72, Speed=@n74)) {samplerType=SAMPLERTYPE_LinearColor, texture=&T_Atlas_DigitalNoise_01}.G
  )
  Multiply(A=@n56, B=Saturate(None=@n58))
  Normalize(VectorInput=Transform(None=Constant3Vector() {constant=<rgba 0 0 1 0>}) {transformSourceType=TRANSFORMSOURCE_Local})
  OneMinus(None=@n27)
  TextureCoordinate()
  TextureSample(UVs=Multiply(A=Time(), B=ScalarParameter() {defaultValue=0.051, parameterName=FlickerSpeed})) {samplerType=SAMPLERTYPE_Masks, texture=/Game/Gadgets/Assets/VinderTech_GliderChute/Textures/T_LightingTimeline_M}
  Time()
 ]
}
